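; PCG subgraph, the core process of a Write Attribute primitive: adds an attribute named by the Attribute parameter to inputs tagged "Input:0".
; The AttributeType constant struct parameter supplies both the value and its type.
; Its extracted Type index drives a Select (Multi) over string constants naming each typed value field, defaulting to DoubleValue.
; The selected name overrides Add Attribute's InputSource, so one node serves every listed type.
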(graph Write_Attribute_CoreProcess
  (params (Attribute :type String :default "NewAttribute") (AttributeType :type (Struct /Script/PCG.PCGMetadataTypesConstantStruct)))
  ;[0 0 0 0.5 12] Primary/Secondary Inputs
  (bind FilterDataByTag_0 (Filter%20Data%20By%20Tag :SelectedTags "Input:0"))
  ;[0 0 0 0.5] Core Process
  (bind CreateConstant_1 (Create%20Constant :AttributeTypes (struct :Type String :Int32Value 5 :StringValue "DoubleValue")))
  (bind CreateConstant_2 (Create%20Constant :AttributeTypes (struct :Type String :Int32Value 5 :StringValue "FloatValue")))
  (bind CreateConstant_3 (Create%20Constant :AttributeTypes (struct :Type String :Int32Value 5 :StringValue "Int32Value")))
  (bind CreateConstant_4 (Create%20Constant :AttributeTypes (struct :Type String :Int32Value 5 :StringValue "IntValue")))
  (bind CreateConstant_5 (Create%20Constant :AttributeTypes (struct :Type String :Int32Value 5 :StringValue "Vector2Value")))
  (bind CreateConstant_6 (Create%20Constant :AttributeTypes (struct :Type String :Int32Value 5 :StringValue "VectorValue")))
  (bind CreateConstant_7 (Create%20Constant :AttributeTypes (struct :Type String :Int32Value 5 :StringValue "Vector4Value")))
  (bind CreateConstant_8 (Create%20Constant :AttributeTypes (struct :Type String :Int32Value 5 :StringValue "QuatValue")))
  (bind CreateConstant_9 (Create%20Constant :AttributeTypes (struct :Type String :Int32Value 5 :StringValue "TransformValue")))
  (bind CreateConstant_10 (Create%20Constant :AttributeTypes (struct :Type String :Int32Value 5 :StringValue "StringValue")))
  (bind CreateConstant_11 (Create%20Constant :AttributeTypes (struct :Type String :Int32Value 5 :StringValue "NameValue")))
  (bind CreateConstant_12 (Create%20Constant :AttributeTypes (struct :Type String :Int32Value 5 :StringValue "BoolValue")))
  (bind CreateConstant_13 (Create%20Constant :AttributeTypes (struct :Type String :Int32Value 5 :StringValue "RotatorValue")))
  (bind CreateConstant_14 (Create%20Constant :AttributeTypes (struct :Type String :Int32Value 5 :StringValue "SoftObjectPathValue")))
  (bind CreateConstant_15 (Create%20Constant :AttributeTypes (struct :Type String :Int32Value 5 :StringValue "SoftClassPathValue")))
  (bind ExtractAttribute_18 (Extract%20Attribute%20at%20Index :bForceOutputAttributeToBeInElementsDomain true :InputSource Type :OutputAttributeName EnumSelection :In AttributeType))
  (bind Select%20%28Multi%29_5 (Select%20%28Multi%29 :EnumSelection (struct :Class /Script/ACLPlugin.ACLCompressionLevel) :IntOptions [0 1 2 3 4 5 6 7 8 9 10 11 12 13 14]))
  (bind AddAttribute_0 (Add%20Attribute :MetadataDomainsMapping (map (None None)) :Attributes AttributeType :OutputTarget-param Attribute))
  (connect FilterDataByTag_0 InsideFilter AddAttribute_0 In)
  (connect Select%20%28Multi%29_5 Out AddAttribute_0 InputSource)
  (connect DefaultInputNode In FilterDataByTag_0 In)
  (connect CreateConstant_2 Out Select%20%28Multi%29_5 %30)
  (connect CreateConstant_1 Out Select%20%28Multi%29_5 %31)
  (connect CreateConstant_12 Out Select%20%28Multi%29_5 %310)
  (connect CreateConstant_13 Out Select%20%28Multi%29_5 %311)
  (connect CreateConstant_11 Out Select%20%28Multi%29_5 %312)
  (connect CreateConstant_14 Out Select%20%28Multi%29_5 %313)
  (connect CreateConstant_15 Out Select%20%28Multi%29_5 %314)
  (connect CreateConstant_3 Out Select%20%28Multi%29_5 %32)
  (connect CreateConstant_4 Out Select%20%28Multi%29_5 %33)
  (connect CreateConstant_5 Out Select%20%28Multi%29_5 %34)
  (connect CreateConstant_6 Out Select%20%28Multi%29_5 %35)
  (connect CreateConstant_7 Out Select%20%28Multi%29_5 %36)
  (connect CreateConstant_8 Out Select%20%28Multi%29_5 %37)
  (connect CreateConstant_9 Out Select%20%28Multi%29_5 %38)
  (connect CreateConstant_10 Out Select%20%28Multi%29_5 %39)
  (connect CreateConstant_1 Out Select%20%28Multi%29_5 Default)
  (connect ExtractAttribute_18 Out Select%20%28Multi%29_5 IntegerSelection)
  (output Out AddAttribute_0))
