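; PCG subgraph that writes a named attribute from secondary inputs onto primary input points via projection, without moving or rotating them.
; Data tagged Input:0 is primary; all other data becomes merged target points with extents overridden by Extent_XY and a separate ExtentZ.
; Primary data is index-stamped, merged for a single Projection that keeps zero-density points, then partitioned back by index and stripped of that temporary attribute.
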(graph Write_Attribute_fromProjection_CoreProcess
  (params (Extent_XY :type Double :default 1e+02) (Attribute :type String) (ExtentZ :type Double :default 5e+03))
  (bind MakeConcrete_3 (Make%20Concrete))
  (bind AttributeMathsOp_0 (Attribute%20Maths%20Op :InputSource1 $Extents :Operation Set :InB Extent_XY))
  (bind AttributeMathsOp_1 (Attribute%20Maths%20Op :InputSource1 $Extents.Z :Operation Set :InB ExtentZ))
  (bind MakeConcrete_2 (Make%20Concrete))
  (bind ToPoint_0 (To%20Point))
  (bind MergePoints_0 (Merge%20Points))
  (bind WriteDataIndex_0 (Write%20Data%20Index :bAddAttribute true :bAddTag false :IndexAttribute WriteAttributeFromProj_dataIndex))
  (bind MakeConcrete_1 (Make%20Concrete))
  (bind ToPoint_1 (To%20Point))
  (bind MergePoints_1 (Merge%20Points))
  (bind Projection_1 (Projection :bKeepZeroDensityPoints true :ProjectionParams (struct :bProjectPositions false :bProjectRotations false :AttributeMode IncludeAttributes) :AttributeList Attribute))
  (bind AttributePartition_0 (Attribute%20Partition :PartitionAttributeSelectors [WriteAttributeFromProj_dataIndex]))
  (bind DeleteAttributes_0 (Delete%20Attributes :SelectedAttributes "WriteAttributeFromProj_dataIndex"))
  ;[0 0 0 0.5 12] Primary/Secondary Inputs
  (bind FilterDataByTag_0 (Filter%20Data%20By%20Tag :SelectedTags "Input:0"))
  (connect MakeConcrete_3 Out AttributeMathsOp_0 InA)
  (connect AttributeMathsOp_0 Out AttributeMathsOp_1 InA)
  (connect Projection_1 Out AttributePartition_0 In)
  (connect AttributePartition_0 Out DeleteAttributes_0 In)
  (connect DefaultInputNode In FilterDataByTag_0 In)
  (connect WriteDataIndex_0 Out MakeConcrete_1 In)
  (connect AttributeMathsOp_1 Out MakeConcrete_2 In)
  (connect FilterDataByTag_0 OutsideFilter MakeConcrete_3 In)
  (connect ToPoint_0 Out MergePoints_0 In)
  (connect ToPoint_1 Out MergePoints_1 In)
  (connect MergePoints_1 Out Projection_1 In)
  (connect MergePoints_0 Out Projection_1 Projection%20Target)
  (connect MakeConcrete_2 Out ToPoint_0 In)
  (connect MakeConcrete_1 Out ToPoint_1 In)
  (connect FilterDataByTag_0 InsideFilter WriteDataIndex_0 In)
  (output Out DeleteAttributes_0))
