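; PCG subgraph that samples a texture at points, spline control points or converted polygons and writes one channel into a named attribute.
; Empty input falls back to a generated point grid, and texture placement combines a transform parameter with the scaled actor transform.
; Normalize remaps color using density min/max across all points; otherwise manual ranges apply.
; Mask picks the channel, which is reduced to one value for @data-domain attributes.
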(graph Write_Attribute_fromTexture_CoreProcess
  (params (Attribute :type String :default "density") (Normalize :type Bool :default true) (Remap_value_in_low :type Double) (Remap_value_in_high :type Double :default 1.0) (Remap_value_out_low :type Double) (Remap_value_out_high :type Double :default 1.0) (Texture_transform :type Transform :default (struct :Scale3D (20.97648513089675 1.0 1.0))) (Texture :type (SoftObject /Script/Engine.Texture2D) :default /PCGPrimitives/Primitives/Write/Ressources/T_gradient_radial.T_gradient_radial) (Mask :type (Enum /Script/GeometryScriptingEditor.EGeometryScriptRGBAChannel)))
  (bind GetSplineControlPoints_0 (Get%20Spline%20Control%20Points))
  (bind ToPoint_0 (To%20Point))
  (bind SampleTexture_0 (Sample%20Texture))
  (bind GetElementsCount_0 (Get%20Element%20Count))
  (bind AttributeReduce_1 (Attribute%20Reduce :Operation Sum))
  (bind CreatePointsGrid_0 (Create%20Points%20Grid :CellSize (1e+03 1e+03 1e+03) :GridExtents (5e+04 5e+04 5e+02)))
  (bind Select_0 (Select))
  (bind FilterDataByType_0 (Filter%20Data%20By%20Type :TargetType (struct :Ids [(struct :Struct /Script/PCG.PCGDataTypeInfoPoint)])))
  (bind FilterDataByType_1 (Filter%20Data%20By%20Type :TargetType (struct :Ids [(struct :Struct /Script/PCG.PCGDataTypeInfoSpline)])))
  (bind FilterDataByType_2 (Filter%20Data%20By%20Type :TargetType (struct :Ids [(struct :Struct /Script/PCG.PCGDataTypeInfoPolygon2D)])))
  (bind Branch_0 (Branch :bOutputToB-param Normalize))
  (bind AttributeRemap_0 (Attribute%20Remap :InputSource $Color :InRangeMax-param Remap_value_in_high :InRangeMin-param Remap_value_in_low :OutRangeMax-param Remap_value_out_high :OutRangeMin-param Remap_value_out_low))
  (bind CopyAttributes_2 (Copy%20Attributes :InputSource $Color :OutputTarget color))
  (bind PointToAttributeSet_0 (Point%20To%20Attribute%20Set))
  (bind BreakVectorAttribute_1 (Break%20Vector%20Attribute))
  (bind Select%20%28Multi%29_0 (Select%20%28Multi%29 :EnumSelection (struct :Class /Script/ACLPlugin.ACLCompressionLevel) :IntOptions [0 1 2 3] :IntegerSelection-param Mask))
  (bind AttributeRemap_1 (Attribute%20Remap :InputSource $Color :OutRangeMax-param Remap_value_out_high :OutRangeMin-param Remap_value_out_low))
  (bind CopyAttributes_4 (Copy%20Attributes :InputSource $Color :OutputTarget color))
  (bind PointToAttributeSet_1 (Point%20To%20Attribute%20Set))
  (bind BreakVectorAttribute_2 (Break%20Vector%20Attribute))
  (bind Select%20%28Multi%29_1 (Select%20%28Multi%29 :EnumSelection (struct :Class /Script/ACLPlugin.ACLCompressionLevel) :IntOptions [0 1 2 3] :IntegerSelection-param Mask))
  (bind CopyAttributes_0 (Copy%20Attributes :InputSource color :OutputTarget-param Attribute))
  (bind AttributeReduce_0 (Attribute%20Reduce :OutputAttributeName temp))
  (bind CopyAttributes_1 (Copy%20Attributes :InputSource temp :OutputTarget temp))
  (bind CopyAttributes_3 (Copy%20Attributes :InputSource temp :OutputTarget-param Attribute))
  ;[0.15 0.15 0.15 0.5] TEXTURE
  (bind CreateConstant_1 (Create%20Constant :AttributeTypes (struct :Type Vector :VectorValue (25.0 25.0 1e+01))))
  (bind GetActorData_0 (Get%20Actor%20Data :Mode GetSinglePoint))
  (bind AttributeMathsOp_0 (Attribute%20Maths%20Op :InputSource1 $Scale :Operation Divide))
  (bind AttributeTransformOp_0 (Attribute%20Transform%20Op :InputSource2 $Transform :InA Texture_transform))
  (bind ExtractAttribute_1 (Extract%20Attribute%20at%20Index :bForceOutputAttributeToBeInElementsDomain true :InputSource $Transform :OutputAttributeName trans))
  (bind GetTextureData_0 (Get%20Texture%20Data :bUseAbsoluteTransform true :bUseAdvancedTiling true :ColorChannel Red :Texture-param Texture))
  ;[0.15 0.15 0.15 0.5] IF attribute is @data domain
  (bind AttributeStringOp_0 (Attribute%20String%20Op :Operation Substring :SearchCase IgnoreCase :InA Attribute))
  (bind CreateConstant_0 (Create%20Constant :AttributeTypes (struct :Type String :StringValue "true")))
  (bind AttributeCompareOp_0 (Attribute%20Compare%20Op :OutputTarget Test))
  (bind ExtractAttribute_0 (Extract%20Attribute%20at%20Index :bForceOutputAttributeToBeInElementsDomain true))
  (bind Branch_2 (Branch))
  ;[0.15 0.15 0.15 0.5] Min / Max over the entire set of point
  (bind MergePoints_0 (Merge%20Points))
  (bind AttributeReduce_2 (Attribute%20Reduce :InputSource $Density :Operation Min))
  (bind AttributeReduce_3 (Attribute%20Reduce :InputSource $Density :Operation Max))
  (reroute in%20points FilterDataByType_0.InsideFilter)
  (reroute in%20polygons FilterDataByType_2.InsideFilter)
  (reroute in%20splines FilterDataByType_1.InsideFilter)
  (connect AttributeStringOp_0 Out AttributeCompareOp_0 InA)
  (connect CreateConstant_0 Out AttributeCompareOp_0 InB)
  (connect GetActorData_0 Out AttributeMathsOp_0 InA)
  (connect CreateConstant_1 Out AttributeMathsOp_0 InB)
  (connect Branch_2 Output%20B AttributeReduce_0 In)
  (connect GetElementsCount_0 Out AttributeReduce_1 In)
  (connect MergePoints_0 Out AttributeReduce_2 In)
  (connect MergePoints_0 Out AttributeReduce_3 In)
  (connect Branch_0 Output%20A AttributeRemap_0 In)
  (connect Branch_0 Output%20B AttributeRemap_1 In)
  (connect AttributeReduce_3 Out AttributeRemap_1 InRangeMax)
  (connect AttributeReduce_2 Out AttributeRemap_1 InRangeMin)
  (connect AttributeMathsOp_0 Out AttributeTransformOp_0 InB)
  (connect SampleTexture_0 Out Branch_0 In)
  (connect ExtractAttribute_0 Out Branch_2 bOutputToB)
  (connect Select%20%28Multi%29_0 Out Branch_2 In)
  (connect Select%20%28Multi%29_1 Out Branch_2 In)
  (connect PointToAttributeSet_0 Out BreakVectorAttribute_1 In)
  (connect PointToAttributeSet_1 Out BreakVectorAttribute_2 In)
  (connect Branch_2 Output%20A CopyAttributes_0 Source)
  (connect in%20points Out CopyAttributes_0 Target)
  (connect in%20splines Out CopyAttributes_0 Target)
  (connect in%20polygons Out CopyAttributes_0 Target)
  (connect AttributeReduce_0 Out CopyAttributes_1 Source)
  (connect Branch_2 Output%20B CopyAttributes_1 Target)
  (connect AttributeRemap_0 Out CopyAttributes_2 Source)
  (connect AttributeRemap_0 Out CopyAttributes_2 Target)
  (connect CopyAttributes_1 Out CopyAttributes_3 Source)
  (connect in%20points Out CopyAttributes_3 Target)
  (connect in%20splines Out CopyAttributes_3 Target)
  (connect in%20polygons Out CopyAttributes_3 Target)
  (connect AttributeRemap_1 Out CopyAttributes_4 Source)
  (connect AttributeRemap_1 Out CopyAttributes_4 Target)
  (connect AttributeCompareOp_0 Out ExtractAttribute_0 In)
  (connect AttributeTransformOp_0 Out ExtractAttribute_1 In)
  (connect Select_0 Out FilterDataByType_0 In)
  (connect Select_0 Out FilterDataByType_1 In)
  (connect Select_0 Out FilterDataByType_2 In)
  (connect DefaultInputNode In GetElementsCount_0 In)
  (connect in%20splines Out GetSplineControlPoints_0 In)
  (connect ExtractAttribute_1 Out GetTextureData_0 Transform)
  (connect Branch_0 Output%20B MergePoints_0 In)
  (connect CopyAttributes_2 Out PointToAttributeSet_0 In)
  (connect CopyAttributes_4 Out PointToAttributeSet_1 In)
  (connect GetTextureData_0 Out SampleTexture_0 BaseTexture)
  (connect GetSplineControlPoints_0 Out SampleTexture_0 Point)
  (connect in%20points Out SampleTexture_0 Point)
  (connect ToPoint_0 Out SampleTexture_0 Point)
  (connect AttributeReduce_1 Out Select_0 bUseInputB)
  (connect CreatePointsGrid_0 Out Select_0 Input%20A)
  (connect DefaultInputNode In Select_0 Input%20B)
  (connect BreakVectorAttribute_1 X Select%20%28Multi%29_0 %30)
  (connect BreakVectorAttribute_1 Y Select%20%28Multi%29_0 %31)
  (connect BreakVectorAttribute_1 Z Select%20%28Multi%29_0 %32)
  (connect BreakVectorAttribute_1 W Select%20%28Multi%29_0 %33)
  (connect BreakVectorAttribute_2 X Select%20%28Multi%29_1 %30)
  (connect BreakVectorAttribute_2 Y Select%20%28Multi%29_1 %31)
  (connect BreakVectorAttribute_2 Z Select%20%28Multi%29_1 %32)
  (connect BreakVectorAttribute_2 W Select%20%28Multi%29_1 %33)
  (connect in%20polygons Out ToPoint_0 In)
  (output Out CopyAttributes_0)
  (output Out CopyAttributes_3))
