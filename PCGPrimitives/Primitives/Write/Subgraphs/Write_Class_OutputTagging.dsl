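; PCG subgraph that tags output data with Class and indexed Class_ID tags from its Class parameter, or only an indexed ID tag when Class is undefined.
; It drops empty data, clears old Input and Class tags, then turns each data index into the ID tag.
; In parallel, it backs up incoming _ID tags, joins them per data, and copies them back as tags to track creation origins through parenting.
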(graph Write_Class_OutputTagging
  (params (Class :type String :default "None"))
  (bind RemoveEmptyData_1 (Remove%20Empty%20Data))
  (bind DataNum_0 (Data%20Count))
  (bind Branch_0 (Branch))
  (bind DeleteTags_2 (Delete%20Tags :SelectedTags "Input:0, Input:1, Class"))
  ;[0 0 0 0.5] Class and Class_ID if defined, else ID
  (bind AttributeCompareOp_2 (Attribute%20Compare%20Op :InA Class))
  (bind AttributeCompareOp_3 (Attribute%20Compare%20Op :InA Class))
  (bind AttributeBooleanOp_1 (Attribute%20Boolean%20Op :Operation Or))
  (bind AttributeStringOp_8 (Attribute%20String%20Op :Operation Replace :In Class))
  (bind AttributeStringOp_0 (Attribute%20String%20Op))
  (bind AttributeStringOp_4 (Attribute%20String%20Op))
  (bind CreateConstant_0 (Create%20Constant :AttributeTypes (struct :Type String :StringValue "ID:") :OutputTarget ID))
  (bind Select_0 (Select))
  (bind AttributeStringOp_2 (Attribute%20String%20Op :InB Class))
  (bind Branch_1 (Branch))
  (bind AddTags_1 (Add%20Tags :TagsToAdd "Cluster"))
  (bind WriteDataIndex_1 (Write%20Data%20Index :bAddAttribute true :bAddTag false))
  (bind AttributeStringOp_3 (Attribute%20String%20Op :InputSource2 @Data.Index :OutputTarget @Data.Index))
  (bind DataAttributesToTags_0 (Data%20Attributes%20To%20Tags :AttributesTagsMapping (map ("Index" @Source)) :bDeleteInputsAfterOperation true))
  ;[0 0 0 0.5] Input Class_ID tags backup to persist onto produced output to track data creation origins through its parenting
  (bind GetTags_1 (Get%20Tags :bExtractTagValues true))
  (bind AttributeFilter_0 (Filter%20Attribute%20Elements :AttributeTypes (struct :Type String :StringValue "_ID") :bUseConstantThreshold true :Operator Substring :TargetAttribute Attribute :ThresholdAttribute @Last))
  (bind SortAttributes_1 (Sort%20Attributes :bUseStableSort false :SortAttributes [(struct :InputSource $Index :SortMethod Descending)]))
  (bind AttributeStringOp_9 (Attribute%20String%20Op :InputSource1 Attribute))
  (bind AttributeStringOp_10 (Attribute%20String%20Op :InputSource1 Attribute :InputSource2 Values))
  (bind WriteDataIndex_2 (Write%20Data%20Index :IndexTag "TagsBackedUpIndex"))
  (bind FilterDataByAttribute_0 (Filter%20Data%20By%20Attribute :FilterMode FilterByValue :FilterOperator NotEqual :TargetAttribute Attribute :Threshold (struct :AttributeTypes (struct :Type String))))
  (bind AttributeReduce_1 (Attribute%20Reduce :bWriteToDataDomain true :InputSource Attribute :Operation Join :OutputAttributeName TagsBackedUp))
  (bind AddAttribute_0 (Add%20Attribute :AttributeTypes (struct :Type String) :OutputTarget @Data.TagsBackedUp))
  (bind Gather_0 (Gather))
  (bind SortTags_0 (Sort%20Data%20By%20Tag%20Value :Tag TagsBackedUpIndex))
  ;[0 0 0 0.5] Write Previous Class_ID to identify the data origin
  (bind CopyAttributes_0 (Copy%20Attributes :InputSource @Data.TagsBackedUp :OutputTarget @Data.TagsBackedUp))
  (bind DataAttributesToTags_1 (Data%20Attributes%20To%20Tags :AttributesTagsMapping (map ("TagsBackedUp" @Source)) :bDeleteInputsAfterOperation true))
  (connect FilterDataByAttribute_0 OutsideFilter AddAttribute_0 In)
  (connect Branch_1 Output%20A AddTags_1 In)
  (connect AttributeStringOp_2 Out AddTags_1 TagsToAdd)
  (connect AttributeCompareOp_2 Out AttributeBooleanOp_1 InA)
  (connect AttributeCompareOp_3 Out AttributeBooleanOp_1 InB)
  (connect GetTags_1 Out AttributeFilter_0 In)
  (connect FilterDataByAttribute_0 InsideFilter AttributeReduce_1 In)
  (connect AttributeStringOp_8 Out AttributeStringOp_0 InA)
  (connect Select_0 Out AttributeStringOp_3 InA)
  (connect WriteDataIndex_1 Out AttributeStringOp_3 InB)
  (connect AttributeStringOp_0 Out AttributeStringOp_4 InA)
  (connect SortAttributes_1 Out AttributeStringOp_9 InA)
  (connect AttributeStringOp_9 Out AttributeStringOp_10 InA)
  (connect AttributeStringOp_9 Out AttributeStringOp_10 InB)
  (connect DataNum_0 Out Branch_0 bOutputToB)
  (connect RemoveEmptyData_1 Out Branch_0 In)
  (connect AttributeBooleanOp_1 Out Branch_1 bOutputToB)
  (connect DeleteTags_2 Out Branch_1 In)
  (connect SortTags_0 Out CopyAttributes_0 Source)
  (connect DataAttributesToTags_0 Out CopyAttributes_0 Target)
  (connect AttributeStringOp_3 Out DataAttributesToTags_0 In)
  (connect CopyAttributes_0 Out DataAttributesToTags_1 In)
  (connect RemoveEmptyData_1 Out DataNum_0 In)
  (connect Branch_0 Output%20B DeleteTags_2 In)
  (connect WriteDataIndex_2 Out FilterDataByAttribute_0 In)
  (connect AttributeReduce_1 Out Gather_0 In)
  (connect AddAttribute_0 Out Gather_0 In2)
  (connect Branch_0 Output%20B GetTags_1 In)
  (connect DefaultInputNode In RemoveEmptyData_1 In)
  (connect AttributeBooleanOp_1 Out Select_0 bUseInputB)
  (connect AttributeStringOp_4 Out Select_0 Input%20A)
  (connect CreateConstant_0 Out Select_0 Input%20B)
  (connect AttributeFilter_0 InsideFilter SortAttributes_1 In)
  (connect Gather_0 Out SortTags_0 In)
  (connect AddTags_1 Out WriteDataIndex_1 In)
  (connect Branch_1 Output%20B WriteDataIndex_1 In)
  (connect AttributeStringOp_10 Out WriteDataIndex_2 In)
  (output Out DataAttributesToTags_1))
