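; PCG subgraph that writes distance from primary "Input:0" points, splines or 2D polygons to the remaining inputs into a chosen attribute.
; Targets become merged points, with splines and polygon outlines processed by a loop subgraph.
; Distance is measured per primary type, optionally flattening Z temporarily for 2D, via a hashed temporary attribute.
; Results are optionally normalized, combined with existing values by a selectable operation, averaged for @data attributes, and optionally min/max remapped.
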(graph Write_DistanceFrom_CoreProcess
  (params (MaximumDistance :type Double :default 1e+02) (OuputAttribute :type String :default "Distance") (AttributeOperation :type (Enum /PCGPrimitives/Primitives/Shared/Resources/EPCG_AttributesMathOpsEnum.EPCG_AttributesMathOpsEnum)) (Normalized :type Bool) (ComputeIn2D :type Bool) (OutputDistanceVector :type Bool) (UseRemap :type Bool) (Remap_output_min :type Double) (Remap_output_max :type Double :default 1.0) (RemapClampToUnitRange :type Bool) (RemapIgnoreValuesOutsideInputRange :type Bool))
  ;[0 0 0 0.5 12] Primary/Secondary Inputs
  (bind FilterDataByTag_0 (Filter%20Data%20By%20Tag :SelectedTags "Input:0"))
  ;[0 0 0 0.5] Core Process
  (bind Branch_3 (Branch :bOutputToB-param UseRemap))
  ;;[0.15 0.15 0.15 0.5] Prepare points to compute distance to
  (bind FilterDataByType_3 (Filter%20Data%20By%20Type :TargetType (struct :Ids [(struct :Struct /Script/PCG.PCGDataTypeInfoPoint)])))
  (bind FilterDataByType_2 (Filter%20Data%20By%20Type :TargetType (struct :Ids [(struct :Struct /Script/PCG.PCGDataTypeInfoPolygon2D)])))
  (bind ToPoint_0 (To%20Point))
  (bind CreateSpline_0 (Create%20Spline :bClosedLoop true :bLinear true :InterpTypeAttribute None))
  (bind FilterDataByType_1 (Filter%20Data%20By%20Type :TargetType (struct :Ids [(struct :Struct /Script/PCG.PCGDataTypeInfoSpline)])))
  (bind Loop_0 (Loop :Graph /PCGPrimitives/Primitives/Write/Subgraphs/Write_DistanceFrom_SplineLoop.Write_DistanceFrom_SplineLoop))
  (bind MergePoints_0 (Merge%20Points :bMergeMetadata false))
  ;;[0.15 0.15 0.15 0.5] ComputeIn2D option
  (bind Branch_4 (Branch :bOutputToB-param ComputeIn2D))
  (bind CopyAttributes_11 (Copy%20Attributes :InputSource $Position.Z :OutputTarget __posZ))
  (bind AttributeMathsOp_11 (Attribute%20Maths%20Op :InputSource1 $Position.Z :Operation Set))
  ;;[0.15 0.15 0.15 0.5] Points
  (bind FilterDataByType_0 (Filter%20Data%20By%20Type :TargetType (struct :Ids [(struct :Struct /Script/PCG.PCGDataTypeInfoPoint)])))
  (bind Distance_1 (Distance :MaximumDistance 2e+03 :OutputAttribute DistanceFromEdge :SourceShape Center :TargetShape Center :bOutputDistanceVector-param OutputDistanceVector :MaximumDistance-param MaximumDistance))
  ;;;[0.15 0.15 0.15 0.5] ComputeIn2D option
  (bind Branch_5 (Branch :bOutputToB-param ComputeIn2D))
  (bind CopyAttributes_12 (Copy%20Attributes :InputSource $Position.Z :OutputTarget __posZ))
  (bind AttributeMathsOp_12 (Attribute%20Maths%20Op :InputSource1 $Position.Z :Operation Set))
  (bind Branch_6 (Branch :bOutputToB-param ComputeIn2D))
  (bind CopyAttributes_7 (Copy%20Attributes :InputSource __posZ :OutputTarget $Position.Z))
  (bind DeleteAttributes_2 (Delete%20Attributes :SelectedAttributes "__posZ"))
  ;;[0.15 0.15 0.15 0.5] Temp Attribute
  (bind HashAttribute_0 (Hash%20Attribute :In OuputAttribute))
  ;;[0.15 0.15 0.15 0.5] Spline
  (bind FilterDataByType_4 (Filter%20Data%20By%20Type :TargetType (struct :Ids [(struct :Struct /Script/PCG.PCGDataTypeInfoSpline)])))
  (bind GetSplineControlPoints_1 (Get%20Spline%20Control%20Points))
  (bind Distance_2 (Distance :SourceShape Center :TargetShape Center :bOutputDistanceVector-param OutputDistanceVector :MaximumDistance-param MaximumDistance))
  (bind CopyAttributes_1 (Copy%20Attributes :InputSource $Position :OutputTarget $Position))
  ;;;[0.15 0.15 0.15 0.5] ComputeIn2D option
  (bind Branch_11 (Branch :bOutputToB-param ComputeIn2D))
  (bind CopyAttributes_18 (Copy%20Attributes :InputSource $Position.Z :OutputTarget __posZ))
  (bind AttributeMathsOp_15 (Attribute%20Maths%20Op :InputSource1 $Position.Z :Operation Set))
  (bind Branch_12 (Branch :bOutputToB-param ComputeIn2D))
  (bind CopyAttributes_17 (Copy%20Attributes :InputSource __posZ :OutputTarget $Position.Z))
  (bind DeleteAttributes_7 (Delete%20Attributes :SelectedAttributes "__posZ"))
  ;;[0.15 0.15 0.15 0.5] Polygon 2D
  (bind FilterDataByType_5 (Filter%20Data%20By%20Type :TargetType (struct :Ids [(struct :Struct /Script/PCG.PCGDataTypeInfoPolygon2D)])))
  (bind ToPoint_2 (To%20Point))
  (bind Distance_0 (Distance :SourceShape Center :TargetShape Center :bOutputDistanceVector-param OutputDistanceVector :MaximumDistance-param MaximumDistance))
  (bind SortAttributes_0 (Sort%20Attributes :bUseStableSort false :SortAttributes [(struct :InputSource $Index :SortMethod Descending)]))
  (bind CreatePolygon2D_1 (Create%20Polygon%202D))
  ;;;[0.15 0.15 0.15 0.5] ComputeIn2D option
  (bind Branch_9 (Branch :bOutputToB-param ComputeIn2D))
  (bind CopyAttributes_16 (Copy%20Attributes :InputSource $Position.Z :OutputTarget __posZ))
  (bind AttributeMathsOp_14 (Attribute%20Maths%20Op :InputSource1 $Position.Z :Operation Set))
  (bind Branch_10 (Branch :bOutputToB-param ComputeIn2D))
  (bind CopyAttributes_15 (Copy%20Attributes :InputSource __posZ :OutputTarget $Position.Z))
  (bind DeleteAttributes_6 (Delete%20Attributes :SelectedAttributes "__posZ"))
  ;;[0.15 0.15 0.15 0.5] Normalize output attrib
  (bind Branch_1 (Branch :bOutputToB-param Normalized))
  (bind AttributeMathsOp_6 (Attribute%20Maths%20Op :Operation Divide :InB MaximumDistance))
  ;;[0.15 0.15 0.15 0.5] Operation Type
  (bind FilterDataByAttribute_0 (Filter%20Data%20By%20Attribute :Attribute-param OuputAttribute))
  (bind Switch_0 (Switch :EnumSelection (struct :Class /PCGPrimitives/Primitives/Shared/Resources/EPCG_AttributesMathOpsEnum.EPCG_AttributesMathOpsEnum) :IntOptions [0 1 2 3 4] :SelectionMode Enum :Value AttributeOperation))
  (bind AttributeMathsOp_5 (Attribute%20Maths%20Op :Operation Set :InputSource1-param OuputAttribute))
  (bind Branch_2 (Branch))
  (bind CopyAttributes_0 (Copy%20Attributes :InputSource @Last :OutputTarget-param OuputAttribute))
  (bind AttributeMathsOp_7 (Attribute%20Maths%20Op :InputSource1-param OuputAttribute))
  (bind AttributeMathsOp_8 (Attribute%20Maths%20Op :Operation Subtract :InputSource1-param OuputAttribute))
  (bind AttributeMathsOp_9 (Attribute%20Maths%20Op :Operation Multiply :InputSource1-param OuputAttribute))
  (bind AttributeMathsOp_10 (Attribute%20Maths%20Op :Operation Divide :InputSource1-param OuputAttribute))
  ;;[0.15 0.15 0.15 0.5] if @data attribute logic, averaging distance value over all points
  (bind AttributeStringOp_1 (Attribute%20String%20Op :Operation Substring :SearchCase IgnoreCase :InA OuputAttribute))
  (bind AttributeReduce_1 (Attribute%20Reduce))
  (bind CopyAttributes_2 (Copy%20Attributes :InputSource @Last :OutputTarget-param OuputAttribute))
  ;;[0.15 0.15 0.15 0.5] Clear temp attrib
  (bind DeleteAttributes_1 (Delete%20Attributes))
  ;;[0.15 0.15 0.15 0.5] BASE RANGE for remap
  (bind Branch_0 (Branch))
  (bind ToPoint_1 (To%20Point))
  (bind PointToAttributeSet_0 (Point%20To%20Attribute%20Set))
  (bind DeleteAttributes_0 (Delete%20Attributes :Operation KeepSelectedAttributes :SelectedAttributes-param OuputAttribute))
  (bind ExtractAttribute_1 (Extract%20Attribute%20at%20Index :bForceOutputAttributeToBeInElementsDomain true :InputSource-param OuputAttribute))
  (bind MergeAttributes_2 (Merge%20Attributes))
  (bind AttributeReduce_0 (Attribute%20Reduce :Operation Min))
  (bind AttributeReduce_2 (Attribute%20Reduce :Operation Max))
  ;;[0.15 0.15 0.15 0.5] Remap
  (bind AttributeRemap_0 (Attribute%20Remap :bClampToUnitRange-param RemapClampToUnitRange :bIgnoreValuesOutsideInputRange-param RemapIgnoreValuesOutsideInputRange :InputSource-param OuputAttribute :OutRangeMax-param Remap_output_max :OutRangeMin-param Remap_output_min))
  (reroute TempAttrib HashAttribute_0)
  (connect Switch_0 Default AttributeMathsOp_5 InA)
  (connect Switch_0 Set AttributeMathsOp_5 InA)
  (connect Switch_0 Default AttributeMathsOp_5 InB)
  (connect Switch_0 Set AttributeMathsOp_5 InB)
  (connect TempAttrib Out AttributeMathsOp_5 InputSource2)
  (connect Branch_1 Output%20B AttributeMathsOp_6 InA)
  (connect TempAttrib Out AttributeMathsOp_6 InputSource1)
  (connect Switch_0 Add AttributeMathsOp_7 InA)
  (connect Switch_0 Add AttributeMathsOp_7 InB)
  (connect TempAttrib Out AttributeMathsOp_7 InputSource2)
  (connect Switch_0 Subtract AttributeMathsOp_8 InA)
  (connect Switch_0 Subtract AttributeMathsOp_8 InB)
  (connect TempAttrib Out AttributeMathsOp_8 InputSource2)
  (connect Switch_0 Multiply AttributeMathsOp_9 InA)
  (connect Switch_0 Multiply AttributeMathsOp_9 InB)
  (connect TempAttrib Out AttributeMathsOp_9 InputSource2)
  (connect Switch_0 Divide AttributeMathsOp_10 InA)
  (connect Switch_0 Divide AttributeMathsOp_10 InB)
  (connect TempAttrib Out AttributeMathsOp_10 InputSource2)
  (connect CopyAttributes_11 Out AttributeMathsOp_11 InA)
  (connect CopyAttributes_12 Out AttributeMathsOp_12 InA)
  (connect CopyAttributes_16 Out AttributeMathsOp_14 InA)
  (connect CopyAttributes_18 Out AttributeMathsOp_15 InA)
  (connect MergeAttributes_2 Out AttributeReduce_0 In)
  (connect Branch_2 Output%20B AttributeReduce_1 In)
  (connect TempAttrib Out AttributeReduce_1 InputSource)
  (connect MergeAttributes_2 Out AttributeReduce_2 In)
  (connect Branch_3 Output%20B AttributeRemap_0 In)
  (connect AttributeReduce_2 Out AttributeRemap_0 InRangeMax)
  (connect AttributeReduce_0 Out AttributeRemap_0 InRangeMin)
  (connect AttributeStringOp_1 Out Branch_0 bOutputToB)
  (connect Branch_3 Output%20B Branch_0 In)
  (connect Branch_6 Output%20A Branch_1 In)
  (connect CopyAttributes_1 Out Branch_1 In)
  (connect CreatePolygon2D_1 Out Branch_1 In)
  (connect DeleteAttributes_2 Out Branch_1 In)
  (connect AttributeStringOp_1 Out Branch_2 bOutputToB)
  (connect FilterDataByAttribute_0 OutsideFilter Branch_2 In)
  (connect DeleteAttributes_1 Out Branch_3 In)
  (connect MergePoints_0 Out Branch_4 In)
  (connect FilterDataByType_0 InsideFilter Branch_5 In)
  (connect Distance_1 Out Branch_6 In)
  (connect ToPoint_2 Out Branch_9 In)
  (connect Distance_0 Out Branch_10 In)
  (connect GetSplineControlPoints_1 Out Branch_11 In)
  (connect Distance_2 Out Branch_12 In)
  (connect TempAttrib Out CopyAttributes_0 InputSource)
  (connect Branch_2 Output%20A CopyAttributes_0 Source)
  (connect Branch_2 Output%20A CopyAttributes_0 Target)
  (connect TempAttrib Out CopyAttributes_1 InputSource)
  (connect TempAttrib Out CopyAttributes_1 OutputTarget)
  (connect Branch_12 Output%20A CopyAttributes_1 Source)
  (connect DeleteAttributes_7 Out CopyAttributes_1 Source)
  (connect FilterDataByType_4 InsideFilter CopyAttributes_1 Target)
  (connect AttributeReduce_1 Out CopyAttributes_2 Source)
  (connect Branch_2 Output%20B CopyAttributes_2 Target)
  (connect Branch_6 Output%20B CopyAttributes_7 Source)
  (connect Branch_6 Output%20B CopyAttributes_7 Target)
  (connect Branch_4 Output%20B CopyAttributes_11 Source)
  (connect Branch_4 Output%20B CopyAttributes_11 Target)
  (connect Branch_5 Output%20B CopyAttributes_12 Source)
  (connect Branch_5 Output%20B CopyAttributes_12 Target)
  (connect Branch_10 Output%20B CopyAttributes_15 Source)
  (connect Branch_10 Output%20B CopyAttributes_15 Target)
  (connect Branch_9 Output%20B CopyAttributes_16 Source)
  (connect Branch_9 Output%20B CopyAttributes_16 Target)
  (connect Branch_12 Output%20B CopyAttributes_17 Source)
  (connect Branch_12 Output%20B CopyAttributes_17 Target)
  (connect Branch_11 Output%20B CopyAttributes_18 Source)
  (connect Branch_11 Output%20B CopyAttributes_18 Target)
  (connect SortAttributes_0 Out CreatePolygon2D_1 In)
  (connect ToPoint_0 Out CreateSpline_0 In)
  (connect PointToAttributeSet_0 Out DeleteAttributes_0 In)
  (connect AttributeMathsOp_5 Out DeleteAttributes_1 In)
  (connect AttributeMathsOp_7 Out DeleteAttributes_1 In)
  (connect AttributeMathsOp_8 Out DeleteAttributes_1 In)
  (connect AttributeMathsOp_9 Out DeleteAttributes_1 In)
  (connect AttributeMathsOp_10 Out DeleteAttributes_1 In)
  (connect CopyAttributes_0 Out DeleteAttributes_1 In)
  (connect CopyAttributes_2 Out DeleteAttributes_1 In)
  (connect TempAttrib Out DeleteAttributes_1 SelectedAttributes)
  (connect CopyAttributes_7 Out DeleteAttributes_2 In)
  (connect CopyAttributes_15 Out DeleteAttributes_6 In)
  (connect CopyAttributes_17 Out DeleteAttributes_7 In)
  (connect TempAttrib Out Distance_0 OutputAttribute)
  (connect AttributeMathsOp_14 Out Distance_0 Source)
  (connect Branch_9 Output%20A Distance_0 Source)
  (connect AttributeMathsOp_11 Out Distance_0 Target)
  (connect Branch_4 Output%20A Distance_0 Target)
  (connect TempAttrib Out Distance_1 OutputAttribute)
  (connect AttributeMathsOp_12 Out Distance_1 Source)
  (connect Branch_5 Output%20A Distance_1 Source)
  (connect AttributeMathsOp_11 Out Distance_1 Target)
  (connect Branch_4 Output%20A Distance_1 Target)
  (connect TempAttrib Out Distance_2 OutputAttribute)
  (connect AttributeMathsOp_15 Out Distance_2 Source)
  (connect Branch_11 Output%20A Distance_2 Source)
  (connect AttributeMathsOp_11 Out Distance_2 Target)
  (connect Branch_4 Output%20A Distance_2 Target)
  (connect Branch_0 Output%20B ExtractAttribute_1 In)
  (connect AttributeMathsOp_6 Out FilterDataByAttribute_0 In)
  (connect Branch_1 Output%20A FilterDataByAttribute_0 In)
  (connect DefaultInputNode In FilterDataByTag_0 In)
  (connect FilterDataByTag_0 InsideFilter FilterDataByType_0 In)
  (connect FilterDataByTag_0 OutsideFilter FilterDataByType_1 In)
  (connect FilterDataByTag_0 OutsideFilter FilterDataByType_2 In)
  (connect FilterDataByTag_0 OutsideFilter FilterDataByType_3 In)
  (connect FilterDataByTag_0 InsideFilter FilterDataByType_4 In)
  (connect FilterDataByTag_0 InsideFilter FilterDataByType_5 In)
  (connect FilterDataByType_4 InsideFilter GetSplineControlPoints_1 In)
  (connect CreateSpline_0 Out Loop_0 In)
  (connect FilterDataByType_1 InsideFilter Loop_0 In)
  (connect DeleteAttributes_0 Out MergeAttributes_2 In)
  (connect ExtractAttribute_1 Out MergeAttributes_2 In)
  (connect FilterDataByType_3 InsideFilter MergePoints_0 In)
  (connect Loop_0 Out MergePoints_0 In)
  (connect ToPoint_1 Out PointToAttributeSet_0 In)
  (connect Branch_10 Output%20A SortAttributes_0 In)
  (connect DeleteAttributes_6 Out SortAttributes_0 In)
  (connect FilterDataByAttribute_0 InsideFilter Switch_0 In)
  (connect FilterDataByType_2 InsideFilter ToPoint_0 In)
  (connect Branch_0 Output%20A ToPoint_1 In)
  (connect FilterDataByType_5 InsideFilter ToPoint_2 In)
  (output Out AttributeRemap_0)
  (output Out Branch_3.Output%20A))
