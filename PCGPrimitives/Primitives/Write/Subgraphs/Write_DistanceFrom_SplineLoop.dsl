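; PCG subgraph that samples an input spline into points at a spacing derived from its size, apparently as a loop body for distance-from-spline writes.
; Spline to Segment feeds a sum of segment X extents, a length measure that is then multiplied and divided.
; Max and Min steps clamp that result before it overrides the Spline Sampler's distance increment.
; The distance-mode, curve-fitted, unbounded sampler's points are the output.
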(graph Write_DistanceFrom_SplineLoop
  (bind SplineToSegment_0 (Spline%20to%20Segment))
  (bind AttributeReduce_0 (Attribute%20Reduce :InputSource $Extents.x :Operation Sum))
  (bind AttributeMathsOp_1 (Attribute%20Maths%20Op :Operation Multiply))
  (bind AttributeMathsOp_2 (Attribute%20Maths%20Op :Operation Divide))
  (bind AttributeMathsOp_4 (Attribute%20Maths%20Op :Operation Max))
  (bind AttributeMathsOp_3 (Attribute%20Maths%20Op :Operation Min))
  (bind SplineSampler_0 (Spline%20Sampler :SamplerParams (struct :Mode Distance :DistanceIncrement 25.0 :bFitToCurve true :InteriorOrientation FollowCurvature :InteriorDensityFalloffCurve (struct :EditorCurveData (struct :Keys [(struct :Value 1.0) (struct :Time 1.0)])) :bUnbounded true :PointSteepness 1.0)))
  (connect AttributeReduce_0 Out AttributeMathsOp_1 InA)
  (connect AttributeMathsOp_1 Out AttributeMathsOp_2 InA)
  (connect AttributeMathsOp_4 Out AttributeMathsOp_3 InA)
  (connect AttributeMathsOp_2 Out AttributeMathsOp_4 InA)
  (connect SplineToSegment_0 Out AttributeReduce_0 In)
  (connect AttributeMathsOp_3 Out SplineSampler_0 DistanceIncrement)
  (connect DefaultInputNode In SplineSampler_0 Spline)
  (connect DefaultInputNode In SplineToSegment_0 In)
  (output Out SplineSampler_0))
