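; PCG core subgraph that writes noise from an input attribute to an output attribute, with configurable mode, range and seed.
; It adds the input attribute where missing, then branches on whether its name contains "@Data".
; Element attributes go straight through Attribute Noise.
; Data-domain values are extracted and merged into one set to receive noise, then partitioned by index and copied back onto the original data.
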(graph Write_Noise_CoreProcess
  (params (InputAttribute :type String :default "$Density") (OuputAttribute :type String :default "Noise") (AttributeOperation :type (Enum /PCGPrimitives/Primitives/Shared/Resources/EPCG_AttributesMathOpsEnum.EPCG_AttributesMathOpsEnum)) (NoiseMin :type Double) (NoiseMax :type Double :default 1.0) (HasCustomSeedSource :type Bool) (CustomSeedSource :type String :default "@Last") (Seed :type Int :default -488115970))
  ;[0.15 0.15 0.15 0.5] IF NO ATTR create it
  (bind FilterDataByAttribute_0 (Filter%20Data%20By%20Attribute :Attribute-param InputAttribute))
  (bind AddAttribute_0 (Add%20Attribute :OutputTarget-param InputAttribute))
  ;[0.15 0.15 0.15 0.5] BRANCH POINT / DATA case
  (bind CreateConstant_0 (Create%20Constant :AttributeTypes (struct :Type String :StringValue "@Data")))
  (bind AttributeStringOp_0 (Attribute%20String%20Op :Operation Substring :SearchCase IgnoreCase :InA InputAttribute))
  (bind Branch_0 (Branch))
  ;[0.7604 0.7604 0.7604 0.5] CASE POINTS
  (bind AttributeNoise_1 (Attribute%20Noise :InputSource $Position :bHasCustomSeedSource-param HasCustomSeedSource :CustomSeedSource-param CustomSeedSource :InputSource-param InputAttribute :Mode-param AttributeOperation :NoiseMax-param NoiseMax :NoiseMin-param NoiseMin :OutputTarget-param OuputAttribute :Seed-param Seed))
  ;[0.15 0.15 0.15 0.5] CASE DATA
  (bind ExtractAttribute_0 (Extract%20Attribute%20at%20Index :bForceOutputAttributeToBeInElementsDomain true :OutputAttributeName temp :InputSource-param InputAttribute))
  (bind MergeAttributes_0 (Merge%20Attributes))
  (bind AttributeNoise_2 (Attribute%20Noise :InputSource temp :OutputTarget temp :bHasCustomSeedSource-param HasCustomSeedSource :CustomSeedSource-param CustomSeedSource :Mode-param AttributeOperation :NoiseMax-param NoiseMax :NoiseMin-param NoiseMin :Seed-param Seed))
  (bind AttributePartition_0 (Attribute%20Partition :PartitionAttributeSelectors [$Index]))
  (bind CopyAttributes_0 (Copy%20Attributes :InputSource temp :OutputTarget-param OuputAttribute))
  (connect FilterDataByAttribute_0 OutsideFilter AddAttribute_0 In)
  (connect Branch_0 Output%20A AttributeNoise_1 In)
  (connect MergeAttributes_0 Out AttributeNoise_2 In)
  (connect AttributeNoise_2 Out AttributePartition_0 In)
  (connect CreateConstant_0 Out AttributeStringOp_0 InB)
  (connect AttributeStringOp_0 Out Branch_0 bOutputToB)
  (connect AddAttribute_0 Out Branch_0 In)
  (connect FilterDataByAttribute_0 InsideFilter Branch_0 In)
  (connect AttributePartition_0 Out CopyAttributes_0 Source)
  (connect Branch_0 Output%20B CopyAttributes_0 Target)
  (connect Branch_0 Output%20B ExtractAttribute_0 In)
  (connect DefaultInputNode In FilterDataByAttribute_0 In)
  (connect ExtractAttribute_0 Out MergeAttributes_0 In)
  (output Out AttributeNoise_1)
  (output Out CopyAttributes_0))
